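# Translucent unlit Niagara material that draws the face effect on the EctoCat backpack cosmetic.
# Sine-driven twist, controlled by particle dynamic parameters, rotates vertices around the particle position and pushes them toward the camera.
# Panning cloud noise distorts the face texture UVs, with distortion modulated by the texture's blue channel.
# A radial gradient blends two particle-tinted emissive colors, while red and green face channels form an alpha faded by viewing angle.
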Material /BRCosmetics/Accessories/FORT_Backpacks/Backpack_M_MED_EctoCat/FX/M_EctoCat_Face
graph {
 properties {
  bUsedWithNiagaraMeshParticles=true
  bUsedWithNiagaraSprites=true
  blendMode=BLEND_Translucent
  pixelDepthOffsetMode=PDOM_Legacy
  shadingModel=MSM_Unlit
  twoSided=true
  usedShadingModels=Unlit
 }
 nodes [
  n1 = DotProduct(A=CameraVectorWS(), B=Transform(None=Constant3Vector() {constant=<rgba 0 1 0 0>}) {transformSourceType=TRANSFORMSOURCE_Instance})
  Abs(None=@n1)
  Add(A=@n1)
  n2 = Constant3Vector() {constant=<rgba 0 0 1 0>}
  n3 = ComponentMask(None=TextureCoordinate()) {g=true}
  n4 = ScalarParameter() {defaultValue=0.5, parameterName=WaveAmount}
  n5 = DynamicParameter() {paramNames [FaceRotation,Twist,Time,Param4]}
  n6 = Sine(None=Add(A=Multiply(A=@n3, B=@n4) {constB=0.5}, B=@n5.Time))
  n7 = Multiply(A=ScalarParameter() {defaultValue=0.015, parameterName=Param_1}, B=@n5.Twist)
  n8 = RotateAboutAxis(
   NormalizedRotationAxis=Normalize(VectorInput=Transform(None=@n2) {transformSourceType=TRANSFORMSOURCE_Local})
   RotationAngle=Multiply(A=Multiply(A=@n6, B=@n7) {constB=0.25})
   PivotPoint=ParticlePositionWS()
   Position=WorldPosition().XYZ
  )
  n9 = Subtract(A=CameraPositionWS(), B=Add(A=WorldPosition().XYZ, B=@n8))
  Add(A=Multiply(A=Normalize(VectorInput=@n9), B=ScalarParameter() {defaultValue=5, parameterName=VS Camera Offset}), B=@n8)
  n10 = CameraVectorWS()
  n11 = VertexNormalWS()
  Clamp(None=Power(Base=DotProduct(A=@n10, B=@n11)))
  ConstantBiasScale(None=@n1)
  Constant() {r=0.5}
  n12 = VectorParameter() {defaultValue=<rgba 0.526 0.0437 0.5209 1>}
  n13 = VectorParameter() {parameterName=Param_1, defaultValue=<rgba 1 0.2 0.8667 1>}
  n14 = Multiply(A=MaterialFunctionCall(UVs=NamedRerouteUsage(), Radius=ScalarParameter() {defaultValue=0.4}, Density=ScalarParameter() {defaultValue=1, parameterName=Param_2}) {materialFunction=&RadialGradientExponential}.RadialGradientExponential)
  MaterialFunctionCall(Emissive=Multiply(A=LinearInterpolate(A=@n12.RGB, B=@n13.RGB, Alpha=@n14), B=ParticleColor().RGB)) {materialFunction=&MF_FX_EmissiveAdjustments}
  n15 = ScalarParameter() {defaultValue=0.3, parameterName=Param_5}
  n16 = Multiply(A=TextureCoordinate(), B=ScalarParameter() {defaultValue=1, parameterName=Param_3})
  n17 = AppendVector(A=Constant(), B=ScalarParameter() {defaultValue=0.5, parameterName=Param_4})
  n18 = ComponentMask(None=TextureSample(UVs=Panner(Coordinate=@n16, Speed=@n17) {speedY=0.2}) {samplerType=SAMPLERTYPE_Normal, texture=/Game/Accessories/FORT_Backpacks/Backpack_M_MED_EctoCat/FX/CloudNoise_T_N}.RGB) {g=true, r=true}
  Multiply(
   A=OneMinus(None=MaterialFunctionCall(Radius=@n15) {materialFunction=&RadialGradientExponential}.RadialGradientExponential)
   B=@n18
  )
  n19 = Multiply(A=@n18, B=ScalarParameter() {defaultValue=0.03, parameterName=B_1}) {constB=0.3}
  n20 = Multiply(A=@n19, B=TextureSample() {texture=&/T_Face_EctoCat}.B)
  n21 = ScalarParameter() {defaultValue=0.632, parameterName=Alpha}
  NamedRerouteDeclaration(None=Add(A=TextureCoordinate(), B=LinearInterpolate(A=@n19, B=@n20, Alpha=@n21))) {name=UVDistortion, nodeColor=<rgba 0 0.1094 1 1>}
  Reroute()
  n22 = TextureSample(UVs=NamedRerouteUsage()) {samplerSource=SSM_Clamp_WorldGroupSettings, texture=&T_Face_EctoCat}
  n23 = Multiply(A=@n22.G, B=ScalarParameter() {defaultValue=0.4, parameterName=B})
  n24 = Multiply(A=@n22.R, B=ParticleColor().A)
  n25 = Multiply(A=Power(Base=@n1, Exp=ScalarParameter() {defaultValue=1, parameterName=ViewBasedFadeSensitivity, sliderMax=16}) {constExponent=1}) {constB=4}
  Reroute(None=Multiply(A=Add(A=@n24, B=@n23), B=Clamp(None=@n25)))
  Time()
  VertexNormalWS()
 ]
}
