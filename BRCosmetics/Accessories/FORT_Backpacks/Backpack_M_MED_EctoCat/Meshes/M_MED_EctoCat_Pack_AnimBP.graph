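# Animation Blueprint for the EctoCat back bling, posing its cap bones and offsetting the pack with the wearer's movement.
# Additive bone transforms offset the cap when NotSprinting is true: no sprinting, skydiving, swimming, grinding or vehicle.
# In the front end, a flip-flop and delay toggle that flag periodically instead.
# The wearer's upper-spine bend is remapped into pack rotation and translation, with an extra offset while skydiving.
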AnimBlueprint /BRCosmetics/Accessories/FORT_Backpacks/Backpack_M_MED_EctoCat/Meshes/M_MED_EctoCat_Pack_AnimBP
target_skeleton=&/M_MED_EctoCat_Pack_Skeleton
graphs [
 AnimGraph {nodes [
   "Transform (Modify) Bone - Bone: pack"!(ComponentPose, bAlphaBoolEnabled:bool, Translation=Get backpack_translation(), Rotation=Get backpack_rotate(), Scale) -> (Pose) {
    alphaInputType=Bool
    rotationMode=BMM_Additive
    scaleMode=BMM_Additive
    scaleSpace=BCS_BoneSpace
    translationMode=BMM_Additive
    alphaBoolBlend {blendInTime=0.1, blendOption=ExpOut, blendOutTime=0.2}
    boneToModify {boneName=pack}
    rotation {pitch=5}
    translation {y=-1, z=4}
   }
   n1 = Get NotSprinting() -> (NotSprinting:bool)
   n2 = "Transform (Modify) Bone - Bone: Cap_01"!(ComponentPose=Local To Component(LocalPose=Input Pose()), bAlphaBoolEnabled:bool=@n1, Translation="0,0,0", Rotation="45,0,0", Scale="1,1,1") -> (Pose) {
    alphaInputType=Bool
    rotationMode=BMM_Additive
    rotationSpace=BCS_BoneSpace
    translationMode=BMM_Additive
    translationSpace=BCS_BoneSpace
    alphaBoolBlend {blendInTime=0.3, blendOption=ExpOut, blendOutTime=0.4}
    boneToModify {boneName=Cap_01}
    rotation {pitch=45}
   }
   n3 = "Transform (Modify) Bone - Bone: Cap_03"!(ComponentPose=@n2, bAlphaBoolEnabled:bool=@n1, Translation="0,0,1", Rotation="0,0,0", Scale="1,1,1") -> (Pose) {
    alphaInputType=Bool
    rotationMode=BMM_Additive
    rotationSpace=BCS_BoneSpace
    translationMode=BMM_Additive
    translationSpace=BCS_BoneSpace
    alphaBoolBlend {blendInTime=0.3, blendOption=ExpOut, blendOutTime=0.4}
    boneToModify {boneName=Cap_03}
    translation {z=1}
   }
   n4 = "Transform (Modify) Bone - Bone: Cap_04"!(ComponentPose=@n3, bAlphaBoolEnabled:bool=@n1, Translation="-9,0,0", Rotation="0,0,0", Scale="1,1,1") -> (Pose) {
    alphaInputType=Bool
    rotationMode=BMM_Additive
    rotationSpace=BCS_BoneSpace
    translationMode=BMM_Additive
    translationSpace=BCS_BoneSpace
    alphaBoolBlend {blendInTime=0.3, blendOption=ExpOut, blendOutTime=0.4}
    boneToModify {boneName=Cap_04}
    translation {x=-9}
   }
   n5 = "Transform (Modify) Bone - Bone: Cap_spring"!(ComponentPose=@n4, bAlphaBoolEnabled:bool=@n1, Translation="0,0,0", Rotation="18,0,0", Scale="1,1,1") -> (Pose) {
    alphaInputType=Bool
    rotationMode=BMM_Additive
    rotationSpace=BCS_BoneSpace
    translationMode=BMM_Additive
    translationSpace=BCS_BoneSpace
    alphaBoolBlend {blendInTime=0.3, blendOption=ExpOut, blendOutTime=0.4}
    boneToModify {boneName=Cap_spring}
    rotation {pitch=18}
   }
   n6 = "Transform (Modify) Bone - Bone: pack"!(ComponentPose=@n5, Alpha:float, Translation=Get backpack_translation(), Rotation=Get backpack_rotate(), Scale) -> (Pose) {
    rotationMode=BMM_Additive
    scaleMode=BMM_Additive
    scaleSpace=BCS_BoneSpace
    translationMode=BMM_Additive
    alphaBoolBlend {blendOption=Linear}
    boneToModify {boneName=pack}
   }
   n7 = "Transform (Modify) Bone - Bone: pack"!(ComponentPose=@n6, bAlphaBoolEnabled:bool=Get bIsSkyDiving(), Translation="0,-1,4", Rotation="5,0,0", Scale) -> (Pose) {
    alphaInputType=Bool
    rotationMode=BMM_Additive
    scaleMode=BMM_Additive
    scaleSpace=BCS_BoneSpace
    translationMode=BMM_Additive
    alphaBoolBlend {blendInTime=0.1, blendOption=ExpOut, blendOutTime=0.2}
    boneToModify {boneName=pack}
    rotation {pitch=5}
    translation {y=-1, z=4}
   }
   Output Pose(Result=Component To Local(ComponentPose=@n7)) {layerGroup=None}
  ]}
 EventGraph {nodes [
   Event BlueprintUpdateAnimation()
   > n1 = Branch(Condition=Get bIsInFrontEnd())
   > n2 = Sequence()
   > Do Once(Reset=@n4.then)
   > n3 = Flip Flop!(None:Exec) -> (A:Exec, B:Exec, IsA:bool=~) {is=title K2Node_MacroInstance}
   Set NotSprinting(execute=[@n3.A, @n3.B], NotSprinting:bool=@n3) -> (Output_Get:bool)
   > n4 = Delay(Duration=4.5)
   n5 = GetOwnerSkeletalMeshComponent()
   n6 = BreakRotator(InRot=BreakTransform(InTransform=GetDeltaTransformFromRefPose(self=@n5, BoneName=spine_05, BaseName=spine_04)).Rotation)
   Set backpack_rotate(execute=[@n2.then_1, @n13.then], backpack_rotate=MakeRotator(Roll=MapRangeClamped(Value=@n6.Yaw, InRangeA=0, InRangeB=-30, OutRangeB=30)))
   > Set backpack_translation(backpack_translation=MakeVector(Y=MapRangeClamped(Value=@n6.Yaw, InRangeA=0, InRangeB=-30, OutRangeA=-1, OutRangeB=-4)))
   n7 = Get bIsSkyDiving()
   n8 = Get bIsSurfaceSwimming()
   n9 = Get bIsOnGrindRail()
   n10 = Get bIsSwimSprinting() -> (bIsSwimSprinting:bool)
   n11 = Get bIsTacticalSprinting()
   n12 = Get bIsInVehicle() -> (bIsInVehicle:bool)
   n13 = Set NotSprinting(
    execute=@n1.else
    NotSprinting=AND Boolean!(
     A=NOT Boolean(A=@n11)
     B=NOT Boolean(A=@n7)
     C:bool=NOT Boolean(A=@n10)
     D:bool=NOT Boolean(A=@n8)
     E:bool=NOT Boolean(A=@n9)
     F:bool=NOT Boolean(A=@n12)
    ) -> (ReturnValue)
   )
   n14 = Event OnCharacterCustomizationLoaded()
   > Set BaseMesh(BaseMesh:Skeletal Mesh Component Object Reference=@n14.CustomBaseMesh) -> (Output_Get:Skeletal Mesh Component Object Reference)
   TryGetPawnOwner()
  ]}
]
variables [
 backpack_rotate {type=Rotator}
 backpack_translation {type=Vector}
 BaseMesh {type=SkeletalMeshComponent}
 NotSprinting {type=boolean}
 PelvisLocation {type=Vector}
]
